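# Translucent, unlit, two-sided Niagara sprite and mesh particle material for the Slither backpack FX, blending spline, radial glow and spline ring looks through scalar parameters.
# Splines take color from a curve atlas and are moved by SplineThicken plus a sine wave, tapered along V and steered by dynamic parameters.
# Glow has optional Fresnel; ring color takes dynamic desaturation and brightness.
# Particle color and camera fade apply, with depth fade on higher-quality deferred paths.
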Material /BRCosmetics/Accessories/FORT_Backpacks/Backpack_M_MED_Slither/FX_V2/M_Backpack_Slither_Spline
graph {
 properties {
  bUsedWithNiagaraMeshParticles=true
  bUsedWithNiagaraSprites=true
  blendMode=BLEND_Translucent
  pixelDepthOffsetMode=PDOM_Legacy
  shadingModel=MSM_Unlit
  twoSided=true
  usedShadingModels=Unlit
 }
 nodes [
  Constant()
  LinearInterpolate(A=NamedRerouteUsage(), B=Constant3Vector(), Alpha=ScalarParameter() {group=RadialGlow, parameterName=Use RadialGlow})
  n1 = ScalarParameter() {group=RadialGlow, parameterName=Use RadialGlow}
  n2 = ScalarParameter() {group=SplineRing, parameterName=Use SplineRing}
  n3 = LinearInterpolate(A=LinearInterpolate(A=NamedRerouteUsage(), B=NamedRerouteUsage(), Alpha=@n1), B=NamedRerouteUsage(), Alpha=@n2)
  n4 = ParticleColor()
  n5 = MaterialFunctionCall(Distance=ScalarParameter() {defaultValue=128, parameterName=Camera Fade Distance}) {materialFunction=&Near_Camera_Fade}
  MaterialFunctionCall(Emissive=Multiply(A=Multiply(A=@n3, B=@n4.RGB), B=@n5.Result)) {materialFunction=&MF_FX_EmissiveAdjustments}
  n6 = NamedRerouteUsage()
  n7 = NamedRerouteUsage()
  n8 = Multiply(A=LinearInterpolate(A=LinearInterpolate(A=@n6, B=@n7, Alpha=@n1), B=NamedRerouteUsage(), Alpha=@n2), B=@n4.A)
  n9 = DepthFade(Opacity=@n8, FadeDistance=ScalarParameter() {defaultValue=10, parameterName=FadeDistance})
  n10 = QualitySwitch(Default=@n9, Low=@n8, High=@n9, Medium=@n8, Epic=@n9)
  Multiply(A=ShadingPathSwitch(Default=@n10, Deferred=@n10, Forward=@n8, Mobile=@n8), B=@n5.Result)
  n11 = Constant3Vector()
  n12 = ComponentMask(None=TextureCoordinate()) {g=true}
  n13 = DynamicParameter() {defaultValue=<rgba 0 0 1 1>, paramNames [SplineSineRandom, CenterLineDesaturationTOD, CenterLineMultiplyTOD, Param4]}
  n14 = VectorParameter() {group=Spline, parameterName=Sine, channelNames {a=Strength, b=Power, g=Speed, r=Tile}, defaultValue=<rgba 2 1 3 0>}
  n15 = Add(A=@n13.SplineSineRandom, B=@n14.Tile)
  n16 = Add(A=Time(), B=@n13.SplineSineRandom)
  n17 = Sine(None=Add(A=Multiply(A=@n12, B=@n15), B=Multiply(A=@n16, B=@n14.Speed)))
  n18 = DynamicParameter() {parameterIndex=1, paramNames [System Y Diraction, Param2, Param3, Param4]}
  n19 = Multiply(A=Saturate(None=Subtract(A=@n18.System Y Diraction) {constB=0.3})) {constB=1.4286}
  n20 = LinearInterpolate(
   A=MaterialFunctionCall(Z=@n17) {materialFunction=&MakeFloat3}.Result
   B=MaterialFunctionCall(X=@n17) {materialFunction=&MakeFloat3}.Result
   Alpha=OneMinus(None=Saturate(None=@n19))
  )
  n21 = Power(Base=OneMinus(None=@n12), Exp=@n14.Power)
  n22 = Transform(None=Multiply(A=LinearInterpolate(A=@n11, B=@n20, Alpha=@n21), B=@n14.Strength))
  n23 = Constant3Vector()
  n24 = ScalarParameter() {group=SplineRing, parameterName=Use SplineRing}
  n25 = ScalarParameter() {defaultValue=3, group=Spline, parameterName="WidthBase (S)"}
  n26 = ScalarParameter() {defaultValue=3, group=Spline, parameterName="WidthTip (S)"}
  NamedRerouteDeclaration(None=Add(
    A=LinearInterpolate(A=@n22, B=@n23, Alpha=@n24)
    B=MaterialFunctionCall(WidthBase=@n25, WidthTip=@n26) {materialFunction=&SplineThicken}.WorldPositionOffset
   )) {name=Spline_WPO, nodeColor=<rgba 1 0 0.0703 1>}
  n27 = TextureCoordinate()
  n28 = CurveAtlasRowParameter(CurveTime=OneMinus(None=ComponentMask(None=@n27) {g=true})) {
   atlas=/Game/Athena/LinearColorCurves/VFX_CurveAtlas
   curve=/Game/Accessories/FORT_Backpacks/Backpack_M_MED_Slither/FX/Curve_Backpack_Slither_Shine
   defaultValue=56
   group=Spline
   parameterName=SplineCurve
  }
  NamedRerouteDeclaration(None=@n28) {name=Spline_Emi, nodeColor=<rgba 1 0 0.0703 1>}
  n29 = TextureCoordinate()
  n30 = VectorParameter() {group=RadialGlow, parameterName=RadialGlow, channelNames {a=Multiply, b=Power, g=Density, r=Radius}, defaultValue=<rgba 0.5 1 1 1>}
  n31 = Multiply(
   A=Power(
    Base=MaterialFunctionCall(UVs=@n29, Radius=@n30.Radius, Density=@n30.Radius) {materialFunction=&RadialGradientExponential}.RadialGradientExponential
    Exp=@n30.Power
   )
   B=@n30.Multiply
  )
  n32 = CameraVectorWS()
  n33 = PixelNormalWS()
  n34 = VertexNormalWS()
  n35 = QualitySwitch(Default=@n33, Low=@n34, High=@n33, Medium=@n34, Epic=@n33)
  n36 = ShadingPathSwitch(Default=@n35, Deferred=@n35, Forward=@n34, Mobile=@n34)
  n37 = Saturate(None=Abs(None=DotProduct(A=@n32, B=@n36)))
  n38 = ScalarParameter() {defaultValue=1, group=RadialGlow, parameterName=RadialGlow Fresnel Brightness}
  n39 = LinearInterpolate(A=@n31, B=Multiply(A=@n31, B=Multiply(A=@n37, B=@n38)), Alpha=ScalarParameter() {group=RadialGlow, parameterName=Use RadialGlow Fresnel})
  NamedRerouteDeclaration(None=@n39) {name=RadialGlow_Alpha, nodeColor=<rgba 0.1719 0 1 1>}
  NamedRerouteDeclaration(None=Multiply(A=@n39, B=VectorParameter() {group=RadialGlow, parameterName=RadialGlowColor, defaultValue=<rgba 1 1 1 0>}.RGB)) {name=RadialGlow_Emi, nodeColor=<rgba 0.5469 1 0 1>}
  n40 = ComponentMask(None=Multiply(A=TextureCoordinate()) {constB=0.5}) {r=true}
  n41 = Saturate(None=Multiply(A=Sine(None=@n40)) {constB=2})
  NamedRerouteDeclaration(None=Multiply(A=@n28.A, B=Multiply(A=@n41, B=@n41) {constB=2})) {name=Spline_Alpha, nodeColor=<rgba 1 0 0.0703 1>}
  n42 = VectorParameter() {group=SplineRing, parameterName=RingColor, defaultValue=<rgba 1 0.2083 0.05 1>}
  n43 = DynamicParameter() {defaultValue=<rgba 0 0 1 1>, paramNames [SplineSineRandom, CenterLineDesaturationTOD, CenterLineMultiplyTOD, Param4]}
  NamedRerouteDeclaration(None=Multiply(A=Desaturation(None=@n42.RGB, Fraction=@n43.CenterLineDesaturationTOD), B=@n43.CenterLineMultiplyTOD)) {name=Ring_Emi, nodeColor=<rgba 0 0.5781 1 1>}
  NamedRerouteDeclaration(None=@n42.A) {name=Ring_Alpha, nodeColor=<rgba 0.6172 0 1 1>}
 ]
}
